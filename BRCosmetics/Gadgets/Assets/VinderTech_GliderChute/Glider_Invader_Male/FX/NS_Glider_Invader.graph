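# Niagara system for the Invader glider's exhaust and glow, with every emitter attached to the glider skeletal mesh through SkeletalMeshLocation.
# Two mesh-renderer flame emitters (small flame and light cone) vary spawn rate and lifetime with ForwardDot, and stretch and brighten with an FN_Integrator value.
# Single-particle burst emitters add a glow sprite and a particle light.
# User colors tint the effect, with front-end overrides and night color scaling computed for flames.
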NiagaraSystem /BRCosmetics/Gadgets/Assets/VinderTech_GliderChute/Glider_Invader_Male/FX/NS_Glider_Invader
dependencies {
 usedDataInterfaces [&NiagaraDataInterfaceColorCurve, &NiagaraDataInterfaceCurve]
 usedDynamicInputs [&MakeQuatFromAxisAngle, &SetBoolByFloatComparison]
 usedModules [&SystemState, &FN_Integrator]
}
emitters {
 ExhaustFlame_Lightcone {
  emitterUpdateScript [
   &EmitterState()
   &SpawnRate(
    SpawnRate=&FN_FrontEndFloatModifier(Gameplay=&Lerp_Float(A=16, B=22, Alpha=$User.ForwardDot), Front End=32)
    Use Spawn Probability=false
   )
   /SetVariables(
    Emitter.Color_A:LinearColor=&MultiplyLinearColor(
     Linear Color=&FN_Select_Color_from_Time_Of_Day(
      Morning=$User.Color_01
      Day=$User.Color_01
      Evening=$User.Color_01
      Night=&MultiplyLinearColor(Linear Color=$User.Color_01, Scale Factor=<1 1 0.5 0.1>)
     )
     Scale Factor=&MakeFloat4FromVectorAndFloat(
      "Vector (RGB)"=&FN_FrontEndVector3Modifier(Gameplay=&Lerp_Vector(A=<0.5 0.5 0.5>, B=<1 1 1>, Alpha=$System.FN_Integrator.NormalizedValue), Front End=<1 1 1>)
      "Float (Alpha)"=1
     )
    )
    Emitter.Color_B:LinearColor=&MultiplyLinearColor(
     Linear Color=&FN_Select_Color_from_Time_Of_Day(
      Morning=$User.Color_02
      Day=$User.Color_02
      Evening=$User.Color_02
      Night=&MultiplyLinearColor(Linear Color=$User.Color_02, Scale Factor=<1 1 0.5 0.1>)
     )
     Scale Factor=&MakeFloat4FromVectorAndFloat(
      "Vector (RGB)"=&FN_FrontEndVector3Modifier(Gameplay=&Lerp_Vector(A=<0.5 0.5 0.5>, B=<1 1 1>, Alpha=$System.FN_Integrator.NormalizedValue), Front End=<1 1 1>)
      "Float (Alpha)"=1
     )
    )
   ) {isSetParameters=true, module=SetVariables_#1}
   /SetVariables(Emitter.Rainbow:LinearColor=&ColorFromCurve(
     ColorCurve=DataInterface() {propertyValues {
       AlphaCurve {keys [<at 0 1 RCIM_Cubic>,<at 1 1 RCIM_Cubic>]}
       BlueCurve {keys [
         {arriveTangent=-0.0516, interpMode=RCIM_Cubic, time=0.0241, value=0}
         <at 0.1714 0>
         <at 0.3217 0>
         <at 0.4855 1>
         <at 0.702 1>
         {arriveTangent=-3.3273, interpMode=RCIM_Cubic, leaveTangent=-3.3273, time=0.8688, value=0.8783}
         <at 1 0.0084>
        ]}
       GreenCurve {keys [
         {arriveTangent=4.8745, interpMode=RCIM_Cubic, time=0.0241, value=0}
         <at 0.1714 0.7987>
         <at 0.3217 1>
         <at 0.4855 0.9513>
         <at 0.702 0>
         <at 0.8688 0 RCIM_Cubic>
         <at 1 0>
        ]}
       RedCurve {keys [
         <at 0.0241 1 RCIM_Cubic>
         <at 0.1714 1>
         <at 0.3217 0.2142>
         <at 0.4855 0>
         <at 0.702 0.0966>
         {arriveTangent=3.0316, interpMode=RCIM_Cubic, leaveTangent=3.0316, time=0.8688, value=1}
         <at 1 1>
        ]}
      }}
     CurveIndex=$Emitter.NormalizedLoopAge
    )) {isSetParameters=true}
  ]
  particleSpawnScript [
   /Niagara/Modules/Spawn/Initialization/InitializeParticle(
    Color=<rgba 1 1 3 0.8>
    Lifetime=&FN_FrontEndFloatModifier(Gameplay=&Lerp_Float(A=1, B=0.5, Alpha=$User.ForwardDot), Front End=0.5)
    Mesh Scale=<0.1 0.1 0.125>
    Module.Write Color=true
    Module.Write Lifetime=true
    Module.Write Mass=false
    Module.Write Material Random=false
    Module.Write Position=false
    Module.Write Scale=true
    Module.Write SpriteRotation=false
    Module.Write SpriteSize=false
    Position=&SimulationPosition()
    Sprite Size=<5 5>
   )
   &InitialMeshOrientation(Orientation Axis=<0 1 0>, Orientation Vector=<0 0 -1>, Rotation=<0 0 0.5>, Use Orientation Vector=true, Use Rotation Vector=true) {module=InitialMeshOrientation001}
   &DynamicMaterialParameters(
    Index 0 Param 1=&RandomRangeFloat(Minimum=0, Maximum=1)
    Index 0 Param 4=&Lerp_Float(A=2, B=3, Alpha=$User.ForwardDot)
    Module.Param0WriteEnabled=true
    Module.Param1WriteEnabled=false
    Module.Param2WriteEnabled=false
    Module.Param3WriteEnabled=true
   )
  ]
  particleUpdateScript [
   &ParticleState()
   &SkeletalMeshLocation(
    MaskSpawnGroup=false
    Mesh Sampling Type=::NewEnumerator5
    Orientation Sampling=::NewEnumerator1
    Skeletal Mesh=$User.SkeletalMesh
    Socket Sampling Mode=::NewEnumerator5
    Use Sampled Mesh Attribute as Ribbon ID=false
   )
   &ScaleColor(
    Scale RGB=&FN_FrontEndVector3Modifier(Gameplay=&Lerp_Vector(A=<1 1 1>, B=<1.5 1.5 1.5>, Alpha=$System.FN_Integrator.Value), Front End=<1.5 1.5 1.5>)
    ScaleA=false
    ScaleRGB=true
   )
   &ScaleMeshSize(Scale Factor=&FN_FrontEndVector3Modifier(Gameplay=&Lerp_Vector(A=<1 1 1.8>, B=<1 1 2.5>, Alpha=$System.FN_Integrator.Value), Front End=<1 1 1.4>))
   /SetVariables(
    Particles.Color_01:LinearColor=&MultiplyLinearColor(
     Linear Color=$User.Color_01
     Scale Factor=&MakeFloat4FromVectorAndFloat(
      "Vector (RGB)"=&FN_FrontEndVector3Modifier(Gameplay=&Lerp_Vector(A=<0.5 0.5 0.5>, B=<1 1 1>, Alpha=$System.FN_Integrator.NormalizedValue), Front End=<1 1 1>)
      "Float (Alpha)"=0.05
     )
    )
    Particles.Color_02:LinearColor=&MultiplyLinearColor(
     Linear Color=$User.Color_01
     Scale Factor=&MakeFloat4FromVectorAndFloat(
      "Vector (RGB)"=&FN_FrontEndVector3Modifier(Gameplay=&Lerp_Vector(A=<0.5 0.5 0.5>, B=<1 1 1>, Alpha=$System.FN_Integrator.NormalizedValue), Front End=<1 1 1>)
      "Float (Alpha)"=0.2
     )
    )
   ) {isSetParameters=true, module=SetVariables_#2}
   /SetVariables(Particles.InnerFlameScale:Vector3f=&Multiply_Vector(A=$Particles.Scale, B=<0.32 0.2 0.2>)) {isSetParameters=true, module=SetVariables_#3}
   &FN_Mesh_Attributes(
    Create Scale Attribute=true
    Scale=&Multiply_Vector(A=$Particles.Scale, B=<1 1 0.3>)
    Write to Module Namespace=true
   )
   &FN_Sprite_Attributes(Camera Offset=8, Create Camera Offset Attribute=true, Write to Module Namespace=true)
  ]
  rendererClasses [&NiagaraMeshRendererProperties]
  renderers [{rendererClass=&NiagaraMeshRendererProperties, rendererIndex=0}]
 }
 ExhaustFlame_Small {
  emitterUpdateScript [
   &EmitterState()
   &SpawnRate(
    SpawnRate=&FN_FrontEndFloatModifier(Gameplay=&Lerp_Float(A=16, B=22, Alpha=$User.ForwardDot), Front End=32)
    Use Spawn Probability=false
   )
   /SetVariables(
    Emitter.Color_A:LinearColor=&MultiplyLinearColor(
     Linear Color=&FN_Select_Color_from_Time_Of_Day(
      Morning=$User.Color_01
      Day=$User.Color_01
      Evening=$User.Color_01
      Night=&MultiplyLinearColor(Linear Color=$User.Color_01, Scale Factor=<1 1 0.5 0.5>)
     )
     Scale Factor=&MakeFloat4FromVectorAndFloat(
      "Vector (RGB)"=&FN_FrontEndVector3Modifier(Gameplay=&Lerp_Vector(A=<0.5 0.5 0.5>, B=<1 1 1>, Alpha=$System.FN_Integrator.NormalizedValue), Front End=<1 1 1>)
      "Float (Alpha)"=1
     )
    )
    Emitter.Color_B:LinearColor=&MultiplyLinearColor(
     Linear Color=&FN_Select_Color_from_Time_Of_Day(
      Morning=$User.Color_02
      Day=$User.Color_02
      Evening=$User.Color_02
      Night=&MultiplyLinearColor(Linear Color=$User.Color_02, Scale Factor=<1 1 0.5 0.5>)
     )
     Scale Factor=&MakeFloat4FromVectorAndFloat(
      "Vector (RGB)"=&FN_FrontEndVector3Modifier(Gameplay=&Lerp_Vector(A=<0.5 0.5 0.5>, B=<1 1 1>, Alpha=$System.FN_Integrator.NormalizedValue), Front End=<1 1 1>)
      "Float (Alpha)"=1
     )
    )
   ) {isSetParameters=true, module=SetVariables_#1}
   /SetVariables(Emitter.Rainbow:LinearColor=&MultiplyLinearColor(
     Linear Color=&ColorFromCurve(
      ColorCurve=DataInterface() {propertyValues {
        AlphaCurve {keys [<at 0 1 RCIM_Cubic>,<at 1 1 RCIM_Cubic>]}
        BlueCurve {keys [
          {arriveTangent=-0.0516, interpMode=RCIM_Cubic, time=0.0241, value=0}
          <at 0.1714 0>
          <at 0.3217 0>
          <at 0.4855 1>
          <at 0.702 1>
          {arriveTangent=-3.3273, interpMode=RCIM_Cubic, leaveTangent=-3.3273, time=0.8688, value=0.8783}
          <at 1 0.0084>
         ]}
        GreenCurve {keys [
          {arriveTangent=4.8745, interpMode=RCIM_Cubic, time=0.0241, value=0}
          <at 0.1714 0.7987>
          <at 0.3217 1>
          <at 0.4855 0.9513>
          <at 0.702 0>
          <at 0.8688 0 RCIM_Cubic>
          <at 1 0>
         ]}
        RedCurve {keys [
          <at 0.0241 1 RCIM_Cubic>
          <at 0.1714 1>
          <at 0.3217 0.2142>
          <at 0.4855 0>
          <at 0.702 0.0966>
          {arriveTangent=3.0316, interpMode=RCIM_Cubic, leaveTangent=3.0316, time=0.8688, value=1}
          <at 1 1>
         ]}
       }}
      CurveIndex=$Emitter.NormalizedLoopAge
     )
     Scale Factor=<0.1 0.1 0.1 0.1>
    )) {isSetParameters=true}
  ]
  particleSpawnScript [
   &InitializeParticle(
    Color=<rgba 1 1 3 0.8>
    Lifetime=&FN_FrontEndFloatModifier(Gameplay=&Lerp_Float(A=1, B=0.5, Alpha=$User.ForwardDot), Front End=0.5)
    Mesh Scale=<0.1 0.1 0.1>
    Module.Write Color=true
    Module.Write Lifetime=true
    Module.Write Mass=false
    Module.Write Material Random=false
    Module.Write Position=false
    Module.Write Scale=true
    Module.Write SpriteRotation=false
    Module.Write SpriteSize=false
    Position=&SimulationPosition()
    Sprite Size=<5 5>
   )
   &InitialMeshOrientation(Orientation Axis=<0 1 0>, Orientation Vector=<0 0 -1>, Use Orientation Vector=true, Use Rotation Vector=false) {module=InitialMeshOrientation001}
   &DynamicMaterialParameters(
    Index 0 Param 1=&RandomRangeFloat(Minimum=0, Maximum=1)
    Index 0 Param 4=&Lerp_Float(A=2, B=3, Alpha=$User.ForwardDot)
    Module.Param0WriteEnabled=true
    Module.Param1WriteEnabled=false
    Module.Param2WriteEnabled=false
    Module.Param3WriteEnabled=true
   )
  ]
  particleUpdateScript [
   &ParticleState()
   &SkeletalMeshLocation(
    MaskSpawnGroup=false
    Mesh Sampling Type=::NewEnumerator5
    Orientation Sampling=::NewEnumerator1
    Skeletal Mesh=$User.SkeletalMesh
    Socket Sampling Mode=::NewEnumerator5
    Use Sampled Mesh Attribute as Ribbon ID=false
   )
   &ScaleColor(
    Scale RGB=&FN_FrontEndVector3Modifier(Gameplay=&Lerp_Vector(A=<1 1 1>, B=<1.5 1.5 1.5>, Alpha=$System.FN_Integrator.Value), Front End=<1.5 1.5 1.5>)
    ScaleA=false
    ScaleRGB=true
   )
   &ScaleMeshSize(Scale Factor=&FN_FrontEndVector3Modifier(Gameplay=&Lerp_Vector(A=<1 1 1.8>, B=<1 1 2.5>, Alpha=$System.FN_Integrator.Value), Front End=<1 1 1.4>))
   /SetVariables(
    Particles.Color_01:LinearColor=&MultiplyLinearColor(
     Linear Color=$User.Color_01
     Scale Factor=&MakeFloat4FromVectorAndFloat(
      "Vector (RGB)"=&FN_FrontEndVector3Modifier(Gameplay=&Lerp_Vector(A=<0.5 0.5 0.5>, B=<1 1 1>, Alpha=$System.FN_Integrator.NormalizedValue), Front End=<1 1 1>)
      "Float (Alpha)"=1
     )
    )
    Particles.Color_02:LinearColor=&MultiplyLinearColor(
     Linear Color=$User.Color_02
     Scale Factor=&MakeFloat4FromVectorAndFloat(
      "Vector (RGB)"=&FN_FrontEndVector3Modifier(Gameplay=&Lerp_Vector(A=<0.5 0.5 0.5>, B=<1 1 1>, Alpha=$System.FN_Integrator.NormalizedValue), Front End=<1 1 1>)
      "Float (Alpha)"=1
     )
    )
   ) {isSetParameters=true, module=SetVariables_#2}
   /SetVariables(Particles.InnerFlameScale:Vector3f=&Multiply_Vector(A=$Particles.Scale, B=<0.32 0.2 0.2>)) {isSetParameters=true, module=SetVariables_#3}
   &FN_Mesh_Attributes(
    Create Scale Attribute=true
    Scale=&Multiply_Vector(A=$Particles.Scale, B=<1 1 0.5>)
    Write to Module Namespace=true
   )
   &FN_Sprite_Attributes(Camera Offset=20, Create Camera Offset Attribute=true, Write to Module Namespace=true)
  ]
  rendererClasses [&NiagaraMeshRendererProperties]
  renderers [
   {rendererClass=&NiagaraMeshRendererProperties, rendererIndex=0}
   {rendererClass=&NiagaraMeshRendererProperties, rendererIndex=1}
  ]
 }
 InnerGlow {
  emitterUpdateScript [
   &EmitterState()
   &SpawnBurst_Instantaneous(
    Spawn Count=1
    Spawn Probability=&MakeCustomFloatFromBool(
     Bool=&SetBoolByIntComparison(A=0, B=$Emitter.LoopCount, Comparison Type=::NewEnumerator4)
     True Float=1
     False Float=0
    )
    Use Spawn Probability=true
   )
   /SetVariables(Emitter.Rainbow:LinearColor=&ColorFromCurve(
     ColorCurve=DataInterface() {propertyValues {
       AlphaCurve {keys [<at 0 1 RCIM_Cubic>,<at 1 1 RCIM_Cubic>]}
       BlueCurve {keys [
         {arriveTangent=-0.0516, interpMode=RCIM_Cubic, time=0.0241, value=0}
         <at 0.1714 0>
         <at 0.3217 0>
         <at 0.4855 1>
         <at 0.702 1>
         {arriveTangent=-3.3273, interpMode=RCIM_Cubic, leaveTangent=-3.3273, time=0.8688, value=0.8783}
         <at 1 0.0084>
        ]}
       GreenCurve {keys [
         {arriveTangent=4.8745, interpMode=RCIM_Cubic, time=0.0241, value=0}
         <at 0.1714 0.7987>
         <at 0.3217 1>
         <at 0.4855 0.9513>
         <at 0.702 0>
         <at 0.8688 0 RCIM_Cubic>
         <at 1 0>
        ]}
       RedCurve {keys [
         <at 0.0241 1 RCIM_Cubic>
         <at 0.1714 1>
         <at 0.3217 0.2142>
         <at 0.4855 0>
         <at 0.702 0.0966>
         {arriveTangent=3.0316, interpMode=RCIM_Cubic, leaveTangent=3.0316, time=0.8688, value=1}
         <at 1 1>
        ]}
      }}
     CurveIndex=$Emitter.NormalizedLoopAge
    )) {isSetParameters=true}
  ]
  particleSpawnScript [&InitializeParticle(
    Color=&MultiplyLinearColor(Linear Color=$User.Color_01, Scale Factor=<1 1 1 1>)
    Module.Write Color=true
    Module.Write Lifetime=true
    Module.Write Mass=false
    Module.Write Material Random=false
    Module.Write Position=false
    Module.Write Scale=true
    Module.Write SpriteRotation=false
    Module.Write SpriteSize=true
    Position=&SimulationPosition()
    Sprite Size=<20 20>
   )]
  particleUpdateScript [
   &ParticleState(Kill Particles When Lifetime Has Elapsed=false, Let Infinitely Lived Particles Die When Emitter Deactivates=true)
   &SkeletalMeshLocation(
    "Interpolate Mesh Sampling (Update)"=true
    MaskSpawnGroup=false
    Mesh Sampling Type=::NewEnumerator5
    Orientation Sampling=::NewEnumerator1
    Skeletal Mesh=$User.SkeletalMesh
    Socket Sampling Mode=::NewEnumerator5
    Use Sampled Mesh Attribute as Ribbon ID=false
   )
   &ScaleColor(
    Scale RGB=&MakeCustomVectorFromBool(
     Bool=$User.bIsFrontEndPreview
     True Vector=<1 1 1>
     False Vector=&Lerp_Vector(A=<0.5 0.5 0.5>, B=<1 1 1>, Alpha=$System.FN_Integrator.NormalizedValue)
    )
    ScaleA=false
    ScaleRGB=true
   )
   &ScaleColor(
    Color Value To Scale=$Particles.Color
    Scale Alpha=&FloatFromCurve(FloatCurve=DataInterface() {propertyValues {Curve {keys [
         {arriveTangent=-0.1493, interpMode=RCIM_Cubic, leaveTangent=-0.1493, tangentMode=RCTM_User, time=0, value=0}
         {arriveTangent=1.2974, interpMode=RCIM_Cubic, leaveTangent=1.2974, tangentMode=RCTM_User, time=0.2, value=0.8}
         {arriveTangent=-0.0405, interpMode=RCIM_Cubic, leaveTangent=-0.0405, tangentMode=RCTM_User, time=0.5, value=1}
        ]}}})
    ScaleA=true
    ScaleRGB=true
   ) {module=ScaleColor001}
   &FN_Sprite_Attributes(
    Camera Offset=10
    Color=&MultiplyLinearColor(Linear Color=$Particles.Color, Scale Factor=<1 1 1 1>)
    Create Camera Offset Attribute=true
    Create Color Attribute=true
    Create SpriteSize Attribute=true
    SpriteSize=&Multiply_Vector2(A=$Particles.SpriteSize, B=<2 2>)
    Write to Module Namespace=true
   )
   &DynamicMaterialParameters(
    Index 0 Param 1=1
    Index 1 Param 4=15
    Module.Param0WriteEnabled=true
    Module.Param0WriteEnabled1=true
    Module.Param1WriteEnabled=true
    Module.Param1WriteEnabled1=true
    Module.Param2WriteEnabled=true
    Module.Param2WriteEnabled1=true
    Module.Param3WriteEnabled=true
    Module.Param3WriteEnabled1=true
    Write Parameter Index 1=true
   )
   &FN_Light_Attributes(
    Color=$User.Color_01
    Create Color Attribute=true
    Create Exponent Attribute=true
    Create Position Attribute=true
    Create Radius Attribute=true
    Enable Particle Light Enable=true
    Exponent=3
    Light Enabled=true
    Radius=80
    Scale Color=&FloatFromCurve(FloatCurve=DataInterface() {propertyValues {Curve {keys [
         {arriveTangent=0.1597, interpMode=RCIM_Cubic, leaveTangent=0.1597, tangentMode=RCTM_User, time=0, value=0}
         {arriveTangent=1.7949, interpMode=RCIM_Cubic, leaveTangent=1.7949, tangentMode=RCTM_User, time=0.2223, value=0.6825}
         {arriveTangent=-0.0405, interpMode=RCIM_Cubic, leaveTangent=-0.0405, tangentMode=RCTM_User, time=0.5, value=1}
        ]}}})
   ) {module=FN_Light_Attributes001}
   &CameraOffset(Camera Offset Amount=10)
  ]
  rendererClasses [&NiagaraSpriteRendererProperties]
  renderers [
   {rendererClass=&NiagaraSpriteRendererProperties}
   {rendererClass=&NiagaraSpriteRendererProperties, rendererIndex=1}
  ]
 }
 InnerGlow001 {
  emitterUpdateScript [
   &EmitterState()
   &SpawnBurst_Instantaneous(
    Spawn Count=1
    Spawn Probability=&MakeCustomFloatFromBool(
     Bool=&SetBoolByIntComparison(A=0, B=$Emitter.LoopCount, Comparison Type=::NewEnumerator4)
     True Float=1
     False Float=0
    )
    Use Spawn Probability=true
   )
  ]
  particleSpawnScript [&InitializeParticle(
    Color=&MultiplyLinearColor(Linear Color=$User.Color_01, Scale Factor=<1 1 1 1>)
    Module.Write Color=true
    Module.Write Lifetime=true
    Module.Write Mass=false
    Module.Write Material Random=false
    Module.Write Position=false
    Module.Write Scale=true
    Module.Write SpriteRotation=false
    Module.Write SpriteSize=true
    Position=&SimulationPosition()
    Sprite Size=<20 20>
   )]
  particleUpdateScript [
   &ParticleState(Kill Particles When Lifetime Has Elapsed=false, Let Infinitely Lived Particles Die When Emitter Deactivates=true)
   &SkeletalMeshLocation(
    "Inherit Velocity (Scale)"=0
    "Interpolate Mesh Sampling (Update)"=true
    MaskSpawnGroup=false
    Mesh Sampling Type=::NewEnumerator5
    Sampled Position Offset=<0 5 -5>
    Skeletal Mesh=$User.SkeletalMesh
    Socket Sampling Mode=::NewEnumerator5
    Use Sampled Mesh Attribute as Ribbon ID=false
    Velocity Sampling=::NewEnumerator0
    VelocitySpeedLimit=false
   )
   &FN_Light_Attributes(
    Color=$User.LightColor
    Create Color Attribute=true
    Create Exponent Attribute=true
    Create Position Attribute=true
    Create Radius Attribute=true
    Enable Particle Light Enable=true
    Exponent=1
    Light Enabled=true
    Radius=80
    Scale Color=&FloatFromCurve(
     FloatCurve=DataInterface() {propertyValues {Curve {keys [
         {arriveTangent=0.1597, interpMode=RCIM_Cubic, leaveTangent=0.1597, tangentMode=RCTM_User, time=0, value=0}
         {arriveTangent=1.7949, interpMode=RCIM_Cubic, leaveTangent=1.7949, tangentMode=RCTM_User, time=0.2223, value=0.6825}
         {arriveTangent=-0.0405, interpMode=RCIM_Cubic, leaveTangent=-0.0405, tangentMode=RCTM_User, time=0.5, value=1}
        ]}}}
     Scale Curve=10
    )
   ) {module=FN_Light_Attributes001}
  ]
  rendererClasses [&NiagaraLightRendererProperties]
  renderers [{rendererClass=&NiagaraLightRendererProperties, rendererIndex=0}]
 }
}
summary {
 systemName=NS_Glider_Invader
 emitters [
  {emitterName=ExhaustFlame_Small, rendererClasses [&NiagaraMeshRendererProperties]}
  {emitterName=InnerGlow, rendererClasses [&NiagaraSpriteRendererProperties]}
  {emitterName=ExhaustFlame_Lightcone, rendererClasses [&NiagaraMeshRendererProperties]}
  {emitterName=InnerGlow001, rendererClasses [&NiagaraLightRendererProperties]}
 ]
 userVariables [
  User.bIsFrontEnd {type=NiagaraBool}
  User.bIsFrontEndPreview {type=NiagaraBool}
  User.Color_01 {type=LinearColor, description="", defaultValue {struct=LinearColor, value=<rgba 0.1333 0.651 0.0235 1>}}
  User.Color_02 {type=LinearColor, description="", defaultValue {struct=LinearColor, value=<rgba 0.4789 1 0.3698 1>}}
  User.ForwardDot {type=NiagaraFloat, defaultValue=0, description=""}
  User.LightColor {type=LinearColor, description="", defaultValue {struct=LinearColor, value=<rgba 1 1 1 1>}}
  User.RibbonIDOffset {type=NiagaraInt32}
  User.RightDot {type=NiagaraFloat, defaultValue=0, description=""}
  User.SkeletalMesh {
   type=NiagaraDataInterfaceSkeletalMesh
   defaultValue {
    struct=NiagaraExt_VariableValue_DataInterface
    value {dataInterface=&self:NiagaraDataInterfaceSkeletalMesh_0}
   }
  }
  User.UseRainbow {type=NiagaraFloat, defaultValue=0, description=""}
 ]
}
