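# Niagara contrail system for the male Invader glider, drawing colored ribbon trails from sockets or bones on a skeletal mesh.
# AkiraTrail spawns by distance, inherits mesh motion and locks ribbon heads via a scratch module; AkiraTrail_FE spawns by rate only in the front end.
# AnimTrail builds two short ribbons between socket pairs.
# Front-end modifiers add artificial velocity for menus, while a system timer drives gameplay alpha unless debugging in editor.
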NiagaraSystem /BRCosmetics/Gadgets/Assets/VinderTech_GliderChute/Glider_Invader_Male/FX/NS_Glider_Invader_Contrails
dependencies {
 usedDataInterfaces [&NiagaraDataInterfaceCurve]
 usedDynamicInputs [&Boolean_NotOperation]
 usedModules [
  &self:SystemSpawnScript.SystemScriptSource.SystemScriptGraph.NiagaraNodeAssignment_0.SetVariables_#1
  &SystemState
  &FN_Timer
 ]
}
emitters {
 AkiraTrail {
  emitterUpdateScript [
   &EmitterState()
   /Game/Effects/Niagara/Modules/Spawn/FN_SpawnByEngineDistance(Multiplier=2, Step=100)
  ]
  particleSpawnScript [
   &InitializeRibbon(
    Color=$User.AkiraColor
    Lifetime=0.6
    Module.Write Color=true
    Module.Write Lifetime=true
    Module.Write Mass=true
    Module.Write Material Random=true
    Module.Write Position=false
    Module.Write RibbonTwist=false
    Module.Write RibbonWidth=true
    Ribbon Width=2.5
   )
   /SetVariables(Particles.Stream:NiagaraInt32=&SpawnGroup()) {isSetParameters=true}
   &FN_SetSpawnGroup(Number of Spawn Groups:NiagaraInt32=$Emitter.Spawn Multiplier)
   &AddVelocity(Velocity=&FN_FrontEndVector3Modifier(Gameplay=<0 0 0>, Front End=<0 1000 0>))
  ]
  particleUpdateScript [
   &ParticleState()
   &SkeletalMeshLocation(
    Bone / Socket Index=&Add_Integer(A=$Particles.SpawnGroup, B=6)
    MaskSpawnGroup=false
    Mesh Sampling Type=::NewEnumerator5
    Orientation Sampling=::NewEnumerator1
    Position Sampling=::NewEnumerator2
    Skeletal Mesh=$User.SkeletalMesh
    Socket Sampling Mode=::NewEnumerator5
   )
   &InheritSourceMovement(
    Applied Force Scale=&VectorFromFloat(Value=-0.5)
    ApplyForce=true
    ApplyForceFalloff=false
    ApplyPosition=true
    ApplyPositionFalloff=false
    Module.LimitForceSpeed=false
    Source Velocity:Vector3f=$Particles.SkeletalMeshLocation.SampledVelocity
   )
   &SolveForcesAndVelocity(Clamp Velocity=false, Limit Acceleration=false)
   &self:AkiraTrail_0.NiagaraScratchPadContainer_1.FN_Lock_Ribbon_Head(
    Emitter Count:NiagaraInt32=$Emitter.Count
    Locked Position:NiagaraPosition=$Output.SkeletalMeshLocation.SampledPositionWithoutOffset
    Particle Stream:NiagaraInt32=$Particles.Stream
   )
   &ScaleColor(
    Scale Alpha=&FN_FrontEndFloatModifier(
     Gameplay=&MakeCustomFloatFromBool(
      Bool=$User.IsDebugInEditor
      True Float=1
      False Float=&FN_FrontEndFloatModifier(Gameplay=&Multiply_Float(A=$System.FN_Timer001.Value, B=0.1), Front End=1)
     )
     Front End=1
    )
    ScaleA=true
    ScaleRGB=true
   )
  ]
  rendererClasses [&NiagaraRibbonRendererProperties]
  renderers [{rendererClass=&NiagaraRibbonRendererProperties}]
 }
 AkiraTrail_FE {
  emitterUpdateScript [
   &EmitterState()
   &SpawnRate(
    SpawnRate=&Multiply_Float(A=60, B=&FN_FrontEndFloatModifier(Gameplay=0, Front End=1))
    Use Spawn Probability=false
   )
  ]
  particleSpawnScript [
   &InitializeRibbon(
    Color=$User.AkiraColor
    Lifetime=0.6
    Module.Write Color=true
    Module.Write Lifetime=true
    Module.Write Mass=true
    Module.Write Material Random=true
    Module.Write Position=false
    Module.Write RibbonTwist=false
    Module.Write RibbonWidth=true
    Ribbon Width=2.5
   )
   &FN_SetSpawnGroup(Number of Spawn Groups=2)
   &AddVelocity(Velocity=&FN_FrontEndVector3Modifier(Gameplay=<0 0 0>, Front End=<0 500 0>))
  ]
  particleUpdateScript [
   &ParticleState()
   &SkeletalMeshLocation(
    Bone / Socket Index=&Add_Integer(A=$Particles.SpawnGroup, B=6)
    MaskSpawnGroup=false
    Mesh Sampling Type=::NewEnumerator5
    Orientation Sampling=::NewEnumerator1
    Position Sampling=::NewEnumerator1
    Skeletal Mesh=$User.SkeletalMesh
    Socket Sampling Mode=::NewEnumerator5
   )
   &InheritSourceMovement(
    Applied Force Scale=&VectorFromFloat(Value=-0.2)
    ApplyForce=true
    ApplyForceFalloff=false
    ApplyPosition=true
    ApplyPositionFalloff=false
    Module.LimitForceSpeed=false
    Source Velocity:Vector3f=$Particles.SkeletalMeshLocation.SampledVelocity
   )
   &SolveForcesAndVelocity(Clamp Velocity=false, Limit Acceleration=false)
   &ScaleColor(
    Scale Alpha=&FN_FrontEndFloatModifier(
     Gameplay=&MakeCustomFloatFromBool(
      Bool=$User.IsDebugInEditor
      True Float=1
      False Float=&FN_FrontEndFloatModifier(Gameplay=&Multiply_Float(A=$System.FN_Timer001.Value, B=0.1), Front End=1)
     )
     Front End=1
    )
    ScaleA=true
    ScaleRGB=true
   )
  ]
  rendererClasses [&NiagaraRibbonRendererProperties]
  renderers [{rendererClass=&NiagaraRibbonRendererProperties}]
 }
 AnimTrail {
  emitterUpdateScript [
   &EmitterState()
   /Game/Effects/Niagara/Modules/AnimTrail/FN_SpawnAnimTrail/FN_SpawnAnimTrail(Number of Anim Trails=2, SpawnRate=50) {enabled=false}
   /Game/Effects/Niagara/Modules/AnimTrail/FN_SpawnAnimTrail/FN_SpawnAnimTrail_v6/FN_SpawnAnimTrail(Number of Anim Trails=2) {module=FN_SpawnAnimTrail001}
  ]
  particleSpawnScript [
   &FN_InitializeAnimTrailBoneIndecies(First Bone or Socket Index=2, Second Bone or Socket Index=3) {module=FN_InitializeAnimTrailBoneIndecies002}
   &FN_InitializeAnimTrailBoneIndecies(AnimTrail Selection=1, First Bone or Socket Index=5, Second Bone or Socket Index=4) {module=FN_InitializeAnimTrailBoneIndecies003}
   /Game/Effects/Niagara/Modules/AnimTrail/FN_InitializeAnimTrail/FN_InitializeAnimTrailBoneIndecies(First Bone or Socket Index=2, Second Bone or Socket Index=3) {enabled=false}
   &FN_InitializeAnimTrailBoneIndecies(AnimTrail Selection=1, First Bone or Socket Index=5, Second Bone or Socket Index=4) {enabled=false, module=FN_InitializeAnimTrailBoneIndecies001}
   /Game/Effects/Niagara/Modules/AnimTrail/FN_InitializeAnimTrail/FN_InitializeAnimTrail(AnimTrailWidth=1, Color=$User.Color, Lifetime=0.16)
   &AddVelocity(Velocity=&FN_FrontEndVector3Modifier(Gameplay=<0 0 0>, Front End=<0 2000 0>))
   &AddVelocity(Velocity=&MakeCustomVectorFromBool(Bool=$User.IsDebugInEditor, True Vector=<0 -2000 1>, False Vector=<1 0 0>)) {module=AddVelocity001}
   &CameraOffset(Camera Offset Amount=500)
  ]
  particleUpdateScript [
   &SolveForcesAndVelocity(Clamp Velocity=false, Limit Acceleration=false)
   /Game/Effects/Niagara/Modules/AnimTrail/FN_UpdateAnimTrail/FN_UpdateAnimTrail(New float=1)
   &ScaleColor(
    Scale Alpha=&FN_FrontEndFloatModifier(
     Gameplay=&MakeCustomFloatFromBool(
      Bool=$User.IsDebugInEditor
      True Float=1
      False Float=&FN_FrontEndFloatModifier(Gameplay=&Multiply_Float(A=$System.FN_Timer001.Value, B=0.17), Front End=1)
     )
     Front End=1
    )
    ScaleA=true
    ScaleRGB=true
   )
  ]
  rendererClasses [&NiagaraRibbonRendererProperties]
  renderers [{rendererClass=&NiagaraRibbonRendererProperties}]
 }
}
summary {
 systemName=NS_Glider_Invader_Contrails
 emitters [
  {emitterName=AnimTrail, rendererClasses [&NiagaraRibbonRendererProperties]}
  {emitterName=AkiraTrail, rendererClasses [&NiagaraRibbonRendererProperties]}
  {emitterName=AkiraTrail_FE, rendererClasses [&NiagaraRibbonRendererProperties]}
 ]
 userVariables [
  User.AkiraColor {type=LinearColor, description="", defaultValue {struct=LinearColor, value=<rgba 0.1333 0.651 0.0235 1>}}
  User.bIsFrontEnd {type=NiagaraBool}
  User.bIsFrontEndPreview {type=NiagaraBool}
  User.Color {type=LinearColor, description="", defaultValue {struct=LinearColor, value=<rgba 0.0937 1 0.9901 1>}}
  User.ForwardDot {type=NiagaraFloat, defaultValue=0, description=""}
  User.IsDebugInEditor {type=NiagaraBool}
  User.RibbonIDOffset {type=NiagaraInt32}
  User.RightDot {type=NiagaraFloat, defaultValue=0, description=""}
  User.SkeletalMesh {
   type=NiagaraDataInterfaceSkeletalMesh
   defaultValue {
    struct=NiagaraExt_VariableValue_DataInterface
    value {dataInterface=&self:NiagaraDataInterfaceSkeletalMesh_0}
   }
  }
 ]
}
